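; Day sequence actor for testing far shadows on buildings, driven by data-driven CVars.
; BeginPlay caches the sun's default shadow bias, then binds CVar changes to ValidateCvarChanged.
; Validation runs mapped console commands, toggles the cascade count to refresh, and on remote instances updates sun bias and class-filtered building shadow casting, preferring shadow proxy meshes over static meshes.
; The construction script shows or hides baked clouds by moving them far below the world.
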(event EventBeginPlay
  (Variables|Default|SetDefaultShadowBias (Variables|/Script/Engine.LightComponent|GetShadowBias (Variables|DayNightCycle|GetSunComponent)))
  (Utilities|FlowControl|Delay 1.0)
  (DataDrivenCVar|BindEventtoOnDataDrivenCVarDelegate (EngineSubsystems|GetDataDrivenCVarEngineSubsystem) (EventDispatchers|CreateEvent :!event ValidateCvarChanged))
  (CallFunction|ValidateCvarChanged "All"))

(fn BakedCloudsVisibility
  (Transformation|SetWorldLocation (Variables|Default|GetBakedClouds) (select (Variables|Default|GetShowBakedClouds) "0, 0, 0" "0.000000,0.000000,-10000000.000000") false false))

(fn FarShadowCvarSetup01
  (bind (_keypin _valuepin) (Utilities|Map|ForEachLoop(Map) (Variables|Default|GetConsoleCommandsToEnable01)))
  (Development|ExecuteConsoleCommand (Utilities|String|Append _keypin " " (Utilities|String|ToString(Float) _valuepin))))

(fn InitFarShadow
  (bind _returnvalue (Development|GetConsoleVariableIntValue "FarShadowTest.ClassFilterIndex"))
  (if (> _returnvalue 0)
    _))

(fn RefreshCascadeShadows
  (Development|ExecuteConsoleCommand "r.Shadow.MaxNumFarShadowCascades 0")
  (Development|ExecuteConsoleCommand "r.Shadow.MaxNumFarShadowCascades 1"))

(fn SetFarShadowEnabledOnActor
  (params (Actor :type Actor%20Object%20Reference) (FarShadowEnabled :type Boolean))
  (bind _returnvalue (Actor|GetComponentsByClass Actor "/Script/FortniteGame.ShadowProxyMeshComponent"))
  (if (Utilities|Array|IsValidIndex _returnvalue)
    ; If we have a Shadow Proxy then set it on there
    (for _array_element _returnvalue
      (Rendering|SetCastFarShadow _array_element (Variables|GetFarShadowEnabled)))
    (else
      ; No Shadow Proxy so set on first static mesh comp
      (bind _returnvalue_2 (Actor|GetComponentbyClass Actor "/Script/Engine.StaticMeshComponent"))
      (Utilities|IsValid _returnvalue_2
        (:Is%20Valid
          (Rendering|SetCastFarShadow _returnvalue_2 (Variables|GetFarShadowEnabled)))))))

(fn ConstructionScript
  (|CallParentFunction)
  (CallFunction|BakedCloudsVisibility))

(fn ValidateCvarChanged
  (params (CVarName :type String :default "None"))
  (switch string CVarName
    (:Fort.FarShadow.ClassFilterIndex
      _)
    (:Fort.FarShadow.ShadowBias
      (CallFunction|ValidateFarShadowBias)
      (CallFunction|RefreshCascadeShadows))
    (:All
      (Utilities|FlowControl|Sequence
        (:then_0
          _)
        (:then_1
          _))))
  (Default|ValidateFarShadowonBuildings)
  (CallFunction|FarShadowCvarSetup01)
  (CallFunction|RefreshCascadeShadows))

(fn ValidateFarShadowBias
  (bind _suncomponent (Variables|DayNightCycle|GetSunComponent))
  (Networking|SwitchHasAuthority
    (:Remote
      (bind _returnvalue (Development|GetConsoleVariableFloatValue "Fort.FarShadow.ShadowBias"))
      (LocalVariables|Setl_biasCvar _returnvalue)
      (if (< (LocalVariables|Getl_biasCvar) 0)
        (Rendering|Components|Light|SetShadowBias _suncomponent (Variables|Default|GetDefaultShadowBias))
        (else
          (Rendering|Components|Light|SetShadowBias _suncomponent (LocalVariables|Getl_biasCvar)))))))

(fn ValidateFarShadowOnBuildings
  (Networking|SwitchHasAuthority
    (:Remote
      (bind _returnvalue (Development|GetConsoleVariableIntValue "Fort.FarShadow.ClassFilterIndex"))
      (LocalVariables|Setl_classFilterIndex _returnvalue)
      (Utilities|FlowControl|Sequence
        (:then_0
          (bind _outactors (Actor|GetAllActorsOfClass "/Script/FortniteGame.BuildingSMActor"))
          (Utilities|Set|AddItems (LocalVariables|Getl_actors) _outactors)
          (bind _outactors_2 (Actor|GetAllActorsOfClass "/Script/FortniteGame.BuildingStaticGeo"))
          (Utilities|Set|AddItems (LocalVariables|Getl_actors) _outactors_2))
        (:then_1
          (bind _result (Utilities|Set|ToArray (LocalVariables|Getl_actors)))
          (for _array_element _result
            (switch int (Math|Integer|Clamp(Integer) (LocalVariables|Getl_classFilterIndex) 0 2)
              (:Default
                _)
              (:0
                _)
              (:1
                ;[0.15 0.1355 0.06377 0.5] Set only on filtered building classes (Architecture) to cast far shadow
                (LocalVariables|Setl_wasA false)
                (bind _array_element_2 (Utilities|Array|ForEachLoopwithBreak (LocalVariables|Getl_buildingArchFilterClasses))
                  (:LoopBody
                    (if (Utilities|IsA _array_element _array_element_2)
                      (LocalVariables|Setl_wasA true)
                      (Utilities|FlowControl|Sequence
                        (:then_0
                          _))))
                  (:Completed
                    _))
                (CallFunction|SetFarShadowEnabledonActor _array_element (LocalVariables|Getl_wasA)))
              (:2
                ;[0.05068 0.15 0.03988 0.5] Enable far shadow on all
                (CallFunction|SetFarShadowEnabledonActor _array_element true)))
            ;[0.15 0.05533 0.04096 0.5] Disable all far shadows
            (bind _returnvalue_2 (Actor|GetComponentsByClass _array_element "/Script/Engine.StaticMeshComponent"))
            (for _array_element_3 _returnvalue_2
              (Rendering|SetCastFarShadow _array_element_3 false))))))))
